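# Niagara idle effect for the SleepyUni pickaxe, adding warm golden glows to a moon, stars and a glow-shape mesh.
# Four emitters burst-spawn particles that persist past their lifetime.
# MoonGlow follows a socket each update and StarGlow sits at a fixed offset, both pairing a sprite with a light.
# SmallStarsGlow places nine sprites on sockets at spawn via a shared socket reader, using index-matched arrays for per-star size and facing.
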NiagaraSystem /BRCosmetics/Weapons/FORT_Melee/Pickaxe_SleepyUni/FX/NS_Pickaxe_SleepyUni_Idle
dependencies {
 usedDataInterfaces [
  &NiagaraDataInterfaceArrayPosition
  &NiagaraDataInterfaceArrayFloat2
  &NiagaraDataInterfaceArrayFloat3
  &NiagaraDataInterfaceSocketReader
 ]
 usedDynamicInputs []
 usedModules [
  &self:SystemSpawnScript.SystemScriptSource.SystemScriptGraph.NiagaraNodeAssignment_1.SetVariables_#1
  &SystemState
 ]
}
emitters {
 MoonGlow {
  emitterUpdateScript [
   &EmitterState(
    Inactive Response=::NewEnumerator1
    Life Cycle Mode=::NewEnumerator1
    Loop Behavior=::NewEnumerator1
    Loop Duration Mode=::NewEnumerator1
    UseLoopDelay=false
   )
   &SpawnBurst_Instantaneous(Use Spawn Probability=false, UseLoopCountLimit=false)
  ]
  particleSpawnScript [
   &InitializeParticle(
    Color=<rgba 1 0.619 0 0.1>
    Lifetime=5
    Sprite Size Mode=::NewEnumerator3
    Uniform Sprite Size=40
    UsePositionOffset=false
    Write Material Random=true
   )
   &CameraOffset(Camera Offset Amount=2)
  ]
  particleUpdateScript [
   &ParticleState(Kill Particles When Lifetime Has Elapsed=false)
   &SocketLocation(SamplePosOffset=false, Socket Reader=$System.Socket Reader, Socket Sampling Mode=::NewEnumerator3, VelocitySpeedLimit=false)
   &FN_Light_Attributes(
    Color=<rgba 1 0.6732 0.0866 1>
    Create Color Attribute=true
    Create Exponent Attribute=true
    Create Radius Attribute=true
    Enable Particle Light Enable=true
    Exponent=5
    Light Enabled=true
    Radius=50
    Scale Color=10
   )
  ]
  rendererClasses [&NiagaraSpriteRendererProperties, &NiagaraLightRendererProperties]
  renderers [
   {rendererClass=&NiagaraSpriteRendererProperties}
   {rendererClass=&NiagaraLightRendererProperties, rendererIndex=1}
  ]
 }
 SmallStarsGlow {
  emitterUpdateScript [
   &EmitterState(
    Inactive Response=::NewEnumerator1
    Life Cycle Mode=::NewEnumerator1
    Loop Behavior=::NewEnumerator1
    Loop Duration Mode=::NewEnumerator1
    UseLoopDelay=false
   )
   &SpawnBurst_Instantaneous(Spawn Count=9, Use Spawn Probability=false, UseLoopCountLimit=false)
  ]
  particleSpawnScript [
   &InitializeParticle(
    Color=<rgba 1 0.6551 0 0.5>
    Lifetime=5
    Position Offset=/Niagara/DynamicInputs/Arrays/SelectPositionFromArray(
     Array Sampling Mode:ENiagaraArraySamplingMode=::NewEnumerator1
     Direct Array Index:NiagaraInt32=&ReturnParticleIDAttributes()
     Interpolation:NiagaraFloat=?
     Direct Array Mode:ENiagara_WrapClamp=::NewEnumerator0
     Evaluation Type:ENiagaraRandomnessEvaluation=::NewEnumerator0
     Normalized Interpolation:NiagaraBool=true
     Recalculate Random Each Loop:NiagaraBool=false
     Position Array:NiagaraDataInterfaceArrayPosition=DataInterface() {propertyValues {
       GpuSyncMode=SyncCpuToGpu
       MaxElements=0
       PositionData [
        <2.8452 -0.7238 53.7586>
        <2.7563 5.869 -61.3932>
        <-0.8969 7.3104 -54.6514>
        <2.0449 6.7967 -33.8211>
        <-0.4059 7.1046 -19.2738>
        <-0.5302 6.5935 0.7615>
        <3.2853 6.4201 24.159>
        <0.7468 11.041 27.8564>
        <-1.9344 4.2323 42.7364>
       ]
      }}
     Randomness Mode:ENiagaraRandomnessMode=::NewEnumerator0
     Fixed Random Seed:NiagaraBool=false
     Random Seed:NiagaraInt32=?
    )
    Sprite Size Mode=::NewEnumerator3
    UsePositionOffset=true
    Write Material Random=true
   ) {enabled=false}
   &InitializeParticle(
    Color=<rgba 1 0.6551 0 1>
    Lifetime=5
    Sprite Size=&SelectVector2DFromArray(
     Direct Array Index=&ReturnExecIndex()
     Vector 2D Selection Array=DataInterface() {propertyValues {FloatData [<10 10>, <8 8>, <9 9>, <6 6>, <8 10>, <6 7>, <7 7>, <7 7>, <8 8>]}}
    )
    Sprite Size Mode=::NewEnumerator4
    UsePositionOffset=false
    Write Material Random=true
   ) {module=InitializeParticle001}
   &SocketLocation(
    SamplePosOffset=true
    Sampled Position Offset=<0 0 42.7202>
    Socket Reader=$System.Socket Reader
    Socket Sampling Mode=::NewEnumerator3
    VelocitySpeedLimit=false
   ) {enabled=false}
   &CameraOffset(Camera Offset Amount=4)
   &SocketLocation(
    SamplePosOffset=false
    Socket Index=&Add_Integer(A=1, B=&ReturnExecIndex())
    Socket Reader=$System.Socket Reader
    Socket Sampling Mode=::NewEnumerator3
    VelocitySpeedLimit=false
   ) {module=SocketLocation001}
  ]
  particleUpdateScript [
   &ParticleState(Kill Particles When Lifetime Has Elapsed=false)
   &SpriteFacingAndAlignment(
    Sprite Facing=&SelectVectorFromArray(
     Direct Array Index=&Add_Integer(A=0, B=&ReturnExecIndex())
     Vector Selection Array=DataInterface() {propertyValues {FloatData [
        <0 1 -0.1>
        <-1 0.09 0.05>
        <1 -0.26 0>
        <1 0 0.03>
        <-1 0.28 0>
        <-1 -0.4459 0.02>
        <1.692 -1.0156 -0.02>
        <-1 0.5542 0.28>
        <1 -0.49 -0.08>
       ]}}
    )
    WriteAlignment=false
    WriteFacing=true
   )
  ]
  rendererClasses [&NiagaraSpriteRendererProperties]
  renderers [{rendererClass=&NiagaraSpriteRendererProperties}]
 }
 StarGlow {
  emitterUpdateScript [
   &EmitterState(
    Inactive Response=::NewEnumerator1
    Life Cycle Mode=::NewEnumerator1
    Loop Behavior=::NewEnumerator1
    Loop Duration Mode=::NewEnumerator1
    UseLoopDelay=false
   )
   &SpawnBurst_Instantaneous(Use Spawn Probability=false, UseLoopCountLimit=false)
  ]
  particleSpawnScript [
   &InitializeParticle(
    Color=<rgba 1 0.6551 0 1>
    Lifetime=5
    Position Offset=<0 -6.95 80.7339>
    Sprite Size Mode=::NewEnumerator3
    Uniform Sprite Size=40
    UsePositionOffset=true
    Write Material Random=true
   )
   &SocketLocation(
    SamplePosOffset=true
    Sampled Position Offset=<0 0 42.7202>
    Socket Reader=$System.Socket Reader
    Socket Sampling Mode=::NewEnumerator3
    VelocitySpeedLimit=false
   ) {enabled=false}
   &CameraOffset(Camera Offset Amount=2)
  ]
  particleUpdateScript [
   &ParticleState(Kill Particles When Lifetime Has Elapsed=false)
   &SpriteFacingAndAlignment(Sprite Facing=<0 -1 -0.08>, WriteAlignment=false, WriteFacing=true)
   &FN_Light_Attributes(
    Color=<rgba 1 0.6732 0.0866 1>
    Create Color Attribute=true
    Create Exponent Attribute=true
    Create Position Attribute=true
    Create Radius Attribute=true
    Enable Particle Light Enable=true
    Exponent=4
    Light Enabled=true
    Position=&Add_Vector(A=Linked() {linkedVariable {name=Particles.Position, type=NiagaraPosition}}, B=<0 -13 0>)
    Radius=30
    Scale Color=10
   )
  ]
  rendererClasses [&NiagaraSpriteRendererProperties, &NiagaraLightRendererProperties]
  renderers [
   {rendererClass=&NiagaraSpriteRendererProperties}
   {rendererClass=&NiagaraLightRendererProperties, rendererIndex=1}
  ]
 }
 StarGlowShape {
  emitterUpdateScript [
   &EmitterState(
    Inactive Response=::NewEnumerator1
    Life Cycle Mode=::NewEnumerator1
    Loop Behavior=::NewEnumerator1
    Loop Duration Mode=::NewEnumerator1
    UseLoopDelay=false
   )
   &SpawnBurst_Instantaneous(Use Spawn Probability=false, UseLoopCountLimit=false)
  ]
  particleSpawnScript [
   &InitializeParticle(Color=<rgba 1 0.6551 0 1>, Lifetime=5, UsePositionOffset=false, Write Material Random=true)
   &CameraOffset()
  ]
  particleUpdateScript [&ParticleState(Kill Particles When Lifetime Has Elapsed=false)]
  rendererClasses [&NiagaraMeshRendererProperties]
  renderers [{rendererClass=&NiagaraMeshRendererProperties, rendererIndex=0}]
 }
}
summary {
 systemName=NS_Pickaxe_SleepyUni_Idle
 emitters [
  {emitterName=SmallStarsGlow, rendererClasses [&NiagaraSpriteRendererProperties]}
  {emitterName=StarGlow, rendererClasses [&NiagaraSpriteRendererProperties, &NiagaraLightRendererProperties]}
  {emitterName=MoonGlow, rendererClasses [&NiagaraSpriteRendererProperties, &NiagaraLightRendererProperties]}
  {emitterName=StarGlowShape, rendererClasses [&NiagaraMeshRendererProperties]}
 ]
 userVariables []
}
