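# Physics asset for the SleepyUni two-handed pickaxe, giving collision and joints to the handle and a dangling moon chain.
# A large capsule covers the handle, with small capsules on the base_Moon anchor and three dyn_Moon bones linked in sequence.
# The base-to-handle joint leaves all angular motion free, while each chain joint has a single nonzero swing limit, alternating axes down the chain like hinges.
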PhysicsAsset /BRCosmetics/Weapons/FORT_Melee/Pickaxe_SleepyUni/Meshes/SleepyUni_Axe_Physics
bodies {
 Melee_TwoHanded_Handle [{
   shapeType=Capsule
   length=100.3476
   radius=8.8402
   shapeName=Melee_TwoHanded_Handle_capsule
   center=<0.7207 -2.0601 24.4523>
   rotation=<rot -4.2756 -90 0.0002>
  }]
 base_Moon [{shapeType=Capsule, length=0.505, radius=0.505, shapeName=base_Moon_capsule, rotation=<rot 0 -90 0>}]
 dyn_Moon_01 [{shapeType=Capsule, length=3, radius=2, center=<2 0 0>, rotation=<rot 90 0 0>}]
 dyn_Moon_02 [{shapeType=Capsule, length=3, radius=2, center=<1 0 0>, rotation=<rot 90 0 0>}]
 dyn_Moon_03 [{
   shapeType=Capsule
   length=5.8311
   radius=1.9605
   shapeName=dyn_Moon_03_capsule
   center=<2.4312 -0.4882 -0.0978>
   rotation=<rot 0 -87.3734 -90>
  }]
}
constraints [
 {
  bone1Name=base_Moon
  bone2Name=Melee_TwoHanded_Handle
  swing1LimitDegrees=45
  swing1Motion=Free
  swing2LimitDegrees=45
  swing2Motion=Free
  twistLimitDegrees=45
  twistMotion=Free
 }
 {bone1Name=dyn_Moon_01, bone2Name=base_Moon, swing1LimitDegrees=120, swing2LimitDegrees=0, twistLimitDegrees=0}
 {bone1Name=dyn_Moon_02, bone2Name=dyn_Moon_01, swing1LimitDegrees=0, swing2LimitDegrees=90, twistLimitDegrees=0}
 {bone1Name=dyn_Moon_03, bone2Name=dyn_Moon_02, swing1LimitDegrees=90, swing2LimitDegrees=0, twistLimitDegrees=0}
]
