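# Animation Blueprint for the SleepyUni pickaxe, driving secondary motion on its dynamic brush and moon bones.
# Chained Anim Dynamics nodes simulate the brush, and the cached result feeds both a RigidBody path and a cheaper perf-mode path.
# Perf mode poses the moon bones with bone offsets and trail controllers, and bRigidBodyEnabled selects the output path.
# The update event sets gravity override, simulation space and separate front-end weapon-only and on-pawn perf flags.
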AnimBlueprint /BRCosmetics/Weapons/FORT_Melee/Pickaxe_SleepyUni/Meshes/SleepyUni_Axe_AnimBP
target_skeleton=&/SleepyUni_Axe_Skeleton
graphs [
 AnimGraph {nodes [
   Output Pose(Result=Blend Poses by bool(BlendPose_0=Use cached pose 'PhysicsEnabled'(), BlendPose_1=Use cached pose 'PerfMode'(), bActiveValue=Get bRigidBodyEnabled()) {blendType=Linear}) {layerGroup=None}
   AnimGraphNode_SubInput_0 = Input Pose()
   n1 = Anim Dynamics - Bone: dyn_Brush_01(ComponentPose=Local To Component(LocalPose=@AnimGraphNode_SubInput_0), Alpha=0.4, ExternalForce="0,0,0") {
    angularDampingOverride=1
    angularSpringConstant=6000
    bAngularSpring=true
    bOverrideAngularDamping=true
    bUseGravityOverride=true
    gravityScale=0.35
    simulationSpace=RootRelative
    alphaBoolBlend {blendOption=Linear}
    boundBone {boneName=dyn_Brush_01}
    componentAppliedLinearAccClamp=<3000 3000 3000>
    componentLinearAccScale=<1 1 1>
    componentLinearVelScale=<1 1 1>
    gravityOverride {z=-980}
    physicsBodyDefinitions [{
      boundBone {boneName=dyn_Brush_01}
      boxExtents=<15 30 15>
      constraintSetup {angularLimitsMax=<45 45 45>, angularLimitsMin=<-45 -45 -45>, angularTarget=<1 0 0>}
      localJointOffset=<0.5 0 0>
     }]
   }
   n2 = Anim Dynamics - Bone: dyn_Brush_02(ComponentPose=@n1, Alpha=0.6, ExternalForce="0,0,0") {
    angularDampingOverride=1
    angularSpringConstant=6000
    bAngularSpring=true
    bOverrideAngularDamping=true
    bUseGravityOverride=true
    gravityScale=0.35
    simulationSpace=RootRelative
    alphaBoolBlend {blendOption=Linear}
    boundBone {boneName=dyn_Brush_02}
    componentAppliedLinearAccClamp=<3000 3000 3000>
    componentLinearAccScale=<1 1 1>
    componentLinearVelScale=<1 1 1>
    gravityOverride {z=-980}
    physicsBodyDefinitions [{
      boundBone {boneName=dyn_Brush_02}
      boxExtents=<15 30 15>
      constraintSetup {angularLimitsMax=<45 45 45>, angularLimitsMin=<-45 -45 -45>, angularTarget=<1 0 0>}
      localJointOffset=<0.5 0 0>
     }]
   }
   n3 = Anim Dynamics - Bone: dyn_Brush_01!(SimulationSpace:AnimPhysSimSpaceType Enum, ComponentPose=@n2, Alpha=0.4, ExternalForce="0,0,0") -> (Pose) {
    angularDampingOverride=1
    angularSpringConstant=6000
    bAngularSpring=true
    bOverrideAngularDamping=true
    bUseGravityOverride=true
    gravityScale=0.35
    simulationSpace=RootRelative
    alphaBoolBlend {blendOption=Linear}
    boundBone {boneName=dyn_Brush_01}
    componentAppliedLinearAccClamp=<3000 3000 3000>
    componentLinearAccScale=<1 1 1>
    componentLinearVelScale=<1 1 1>
    gravityOverride {z=-980}
    physicsBodyDefinitions [{
      boundBone {boneName=dyn_Brush_01}
      boxExtents=<15 30 15>
      constraintSetup {angularLimitsMax=<45 45 45>, angularLimitsMin=<-45 -45 -45>, angularTarget=<1 0 0>}
      localJointOffset=<0.5 0 0>
     }]
   }
   n4 = Anim Dynamics - Bone: dyn_Brush_02!(SimulationSpace:AnimPhysSimSpaceType Enum, ComponentPose=@n3, Alpha=0.6, ExternalForce="0,0,0") -> (Pose) {
    angularDampingOverride=1
    angularSpringConstant=6000
    bAngularSpring=true
    bOverrideAngularDamping=true
    bUseGravityOverride=true
    gravityScale=0.35
    simulationSpace=RootRelative
    alphaBoolBlend {blendOption=Linear}
    boundBone {boneName=dyn_Brush_02}
    componentAppliedLinearAccClamp=<3000 3000 3000>
    componentLinearAccScale=<1 1 1>
    componentLinearVelScale=<1 1 1>
    gravityOverride {z=-980}
    physicsBodyDefinitions [{
      boundBone {boneName=dyn_Brush_02}
      boxExtents=<15 30 15>
      constraintSetup {angularLimitsMax=<45 45 45>, angularLimitsMin=<-45 -45 -45>, angularTarget=<1 0 0>}
      localJointOffset=<0.5 0 0>
     }]
   }
   Save cached pose 'WithBrush'(Pose=Component To Local(ComponentPose=@n4)) {cacheName=WithBrush}
   n5 = RigidBody!(ComponentPose=Local To Component(LocalPose=Use cached pose 'WithBrush'()), Alpha, OverrideWorldGravity:Vector="0,0,0") -> (Pose) {
    overridePhysicsAsset=&/SleepyUni_Axe_Physics
    simulationSpace=BaseBoneSpace
    alphaBoolBlend {blendOption=Linear}
    baseBoneRef {boneName=Melee_TwoHanded_Handle}
    componentAppliedLinearAccClamp=<3000 3000 3000>
    componentLinearAccScale=<0.7 0.3 0.3>
    componentLinearVelScale=<0.7 0.4 0.4>
    simSpaceSettings {
     maxAngularAcceleration=3000
     maxAngularVelocity=40
     maxLinearAcceleration=900
     maxLinearVelocity=800
     velocityScaleZ=0.7
     worldAlpha=1
     externalLinearDragV=<2 1 1>
    }
   }
   Save cached pose 'PhysicsEnabled'(Pose=Component To Local(ComponentPose=@n5)) {cacheName=PhysicsEnabled}
   n6 = Get FEWeapOnlyPerfMode() -> (FEWeapOnlyPerfMode:bool)
   n7 = "Transform (Modify) Bone - Bone: dyn_Moon_01"!(
    ComponentPose=Local To Component(LocalPose=Use cached pose 'WithBrush'())
    bAlphaBoolEnabled:bool=@n6
    Translation="0,0,0"
    Rotation="0,-40,0"
    Scale="1,1,1"
   ) -> (Pose) {
    alphaInputType=Bool
    rotationMode=BMM_Additive
    rotationSpace=BCS_BoneSpace
    translationMode=BMM_Additive
    translationSpace=BCS_BoneSpace
    alphaBoolBlend {blendOption=Linear}
    boneToModify {boneName=dyn_Moon_01}
    rotation {yaw=-40}
   }
   n8 = "Transform (Modify) Bone - Bone: dyn_Moon_02"!(ComponentPose=@n7, bAlphaBoolEnabled:bool=@n6, Translation="0,0,0", Rotation="0,0,0", Scale="1,1,1") -> (Pose) {
    alphaInputType=Bool
    rotationMode=BMM_Additive
    rotationSpace=BCS_BoneSpace
    translationMode=BMM_Additive
    translationSpace=BCS_BoneSpace
    alphaBoolBlend {blendInTime=0.1, blendOption=Linear, blendOutTime=0.1}
    boneToModify {boneName=dyn_Moon_02}
   }
   n9 = Get OnPawnPerfMode() -> (OnPawnPerfMode:bool)
   n10 = "Transform (Modify) Bone - Bone: dyn_Moon_01"(ComponentPose=@n8, bAlphaBoolEnabled=@n9, Translation="0,0,0", Rotation="0,0,0", Scale="1,1,1") {
    alphaInputType=Bool
    rotationMode=BMM_Additive
    rotationSpace=BCS_BoneSpace
    translationMode=BMM_Additive
    translationSpace=BCS_BoneSpace
    alphaBoolBlend {blendOption=Linear}
    boneToModify {boneName=dyn_Moon_01}
   }
   n11 = "Transform (Modify) Bone - Bone: dyn_Moon_02"(ComponentPose=@n10, bAlphaBoolEnabled=@n9, Translation="0,0,0", Rotation="50,0,0", Scale="1,1,1") {
    alphaInputType=Bool
    rotationMode=BMM_Additive
    rotationSpace=BCS_BoneSpace
    translationMode=BMM_Additive
    translationSpace=BCS_BoneSpace
    alphaBoolBlend {blendInTime=0.1, blendOption=Linear, blendOutTime=0.1}
    boneToModify {boneName=dyn_Moon_02}
    rotation {pitch=50}
   }
   n12 = Trail controller - Bone: dyn_Moon_03!(ComponentPose=@n11, bAlphaBoolEnabled:bool=Get FEWeapOnlyPerfMode()) -> (Pose) {
    alphaInputType=Bool
    rotationLimits [{limitMax=<180 180 180>}, {limitMax=<180 180 180>}]
    rotationOffsets [<0 0 0>, <0 0 0>]
    trailBone {boneName=dyn_Moon_03}
   }
   n13 = Trail controller - Bone: dyn_Moon_04!(ComponentPose=@n12, bAlphaBoolEnabled:bool=@n9) -> (Pose) {
    alphaInputType=Bool
    bUsePlanarLimit=true
    relaxationSpeedScale=2
    baseJoint {boneName=base_Moon}
    planarLimits [{drivingBone {boneName=dyn_Moon_02}, planeTransform {location=<0 1 1>, rotation=<rot 0 0 0>, scale=<1 1 1>}}]
    rotationLimits [{limitMax=<180 180 180>}, {limitMax=<180 180 180>}]
    rotationOffsets [<0 0 0>, <0 0 0>]
    trailBone {boneName=dyn_Moon_04}
   }
   Save cached pose 'PerfMode'(Pose=Component To Local(ComponentPose=@n13)) {cacheName=PerfMode}
  ]}
 EventGraph {nodes [
   Event BlueprintUpdateAnimation()
   > n1 = Sequence!(execute) -> (then_0, then_1, then_2:Exec)
   > n2 = "Gravity Override State Function (Nativized)"!(
    execute
    CharacterPartAnimInstance:Custom Character Part Anim Instance Object Reference=Self-Reference()
    GravityOverrideParameterSet:Fort Gravity Override Parameter Set Structure=Get Gravity Override Params() -> (Gravity Override Params:Fort Gravity Override Parameter Set Structure)
    RigidBodyAnimNodeParameterSet:Fort Rigid Body Anim Node Parameter Set Structure
    DebugDraw:bool=false
    GameTimeInSeconds:float=GetGameTimeInSeconds()
    DebugJointName:Name=root
    StateLogicLayerName:Name=none
   ) -> (
    then
    OutGravityOverride:Vector
    OutCurrentState:Name=None
    OutGravityOverrideParameters:Fort Gravity Override Parameters Structure
    OutRigidBodyAnimNodeParameters:Fort Rigid Body Anim Node Parameters Structure
   )
   > Set GravityOverride(GravityOverride=@n2)
   n3 = Branch(execute=@n1.then_1, Condition=Get bIsInFrontEnd())
   > Set SimlationSpace(SimlationSpace:AnimPhysSimSpaceType Enum=World) -> (Output_Get:AnimPhysSimSpaceType Enum)
   > Set FrontEndAnimDynAlpha(FrontEndAnimDynAlpha:double=0.1) -> (Output_Get:double)
   Set SimlationSpace(execute=@n3.else)
   n4 = Get bRigidBodyEnabled()
   n5 = Get bIsOnPawn() -> (bIsOnPawn:bool)
   Set FEWeapOnlyPerfMode(
    execute=@n1.then_2
    FEWeapOnlyPerfMode:bool=AND Boolean!(A=NOT Boolean(A=@n5), B=Get bIsInFrontEnd(), C:bool=NOT Boolean(A=@n4)) -> (ReturnValue)
   ) -> (Output_Get:bool)
   n6 = Get bRigidBodyEnabled()
   > Set OnPawnPerfMode(OnPawnPerfMode:bool=AND Boolean!(A=Get bIsOnPawn(), B=NOT Boolean(A=@n6)) -> (ReturnValue)) -> (Output_Get:bool)
  ]}
]
variables [
 FEWeapOnlyPerfMode {type=boolean, default=true}
 FrontEndAnimDynAlpha {type=number}
 Gravity Override Params {
  type=FortGravityOverrideParameterSet
  default=(Skydive_GoingUp=(GravityOverride=()),Skydive_FloatingDown=(GravityOverride=()),Skydive_DivingDown=(GravityOverride=()),Skydive_Parachuting=(GravityOverride=()),OnGround_Standing=(GravityOverride=()),OnGround_Moving=(GravityOverride=()),OnGround_Moving_HighRotationRate=(GravityOverride=()),OnGround_Crouch_Moving=(GravityOverride=()),Emote_or_Melee=(GravityOverride=()),Swimming=(GravityOverride=()),Front_End=(GravityOverride=()),DBNO=(GravityOverride=()),DBNO_Carried=(GravityOverride=()),OnGround_Falling=(GravityOverride=()),Ziplining=(GravityOverride=()),Skydive_Surfing=(GravityOverride=()),Riding_Standing=(GravityOverride=()),Riding_Moving=(GravityOverride=()),Riding_Moving_Targeting=(GravityOverride=()),Tactical_Sprint=(GravityOverride=()),Rail_Grinding=(GravityOverride=()),)
 }
 GravityOverride {type=Vector}
 OnPawnPerfMode {type=boolean}
 SimlationSpace {type=AnimPhysSimSpaceType}
]
